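# Physics asset for the combined Elimination Bot Mark II gadget skeletal mesh, defining its simplified collision bodies and joints.
# Five capsules cover the animated root, back, arms master and left and right guns.
# The back and arms attach to the root, and both guns attach to the arms.
# All four constraints share uniform swing and twist limit angles.
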PhysicsAsset /Game/Gadgets/Assets/Elimination_Bot_Mark_II/Meshes/SK_Elimination_Bot_Combined_Mark_II_PhysicsAsset
bodies {
 b_Arms_Mstr [{shapeType=Capsule, length=15.4249, radius=5.0208, center=<0.2262 -0.5534 -1.1824>, rotation=<rot 0 -89.9188 -89.7924>}]
 b_Bot_Back [{shapeType=Capsule, length=17.2936, radius=15.7824, center=<-0.2707 -2.9141 0.3485>, rotation=<rot 29.014 -110.9024 -90>}]
 b_Gun_l [{shapeType=Capsule, length=16.4146, radius=10.3539, center=<0.0872 0.2146 -3.9662>, rotation=<rot 0 0.1927 -85.1156>}]
 b_Gun_r [{shapeType=Capsule, length=16.4146, radius=10.3539, center=<0.3531 -0.2146 3.9662>, rotation=<rot 0 0.1927 -85.1156>}]
 b_Root_Anim [{shapeType=Capsule, length=27.0353, radius=26.4083, center=<-0.6274 9.1345 9.7162>, rotation=<rot 32.9307 -90 0.8393>}]
}
constraints [
 {bone1Name=b_Bot_Back, bone2Name=b_Root_Anim, swing1LimitDegrees=45, swing2LimitDegrees=45, twistLimitDegrees=45}
 {bone1Name=b_Arms_Mstr, bone2Name=b_Root_Anim, swing1LimitDegrees=45, swing2LimitDegrees=45, twistLimitDegrees=45}
 {bone1Name=b_Gun_l, bone2Name=b_Arms_Mstr, swing1LimitDegrees=45, swing2LimitDegrees=45, twistLimitDegrees=45}
 {bone1Name=b_Gun_r, bone2Name=b_Arms_Mstr, swing1LimitDegrees=45, swing2LimitDegrees=45, twistLimitDegrees=45}
]
